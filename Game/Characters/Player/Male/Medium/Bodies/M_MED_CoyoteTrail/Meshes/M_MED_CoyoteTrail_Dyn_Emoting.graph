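# Physics asset for the CoyoteTrail male medium body, an emote variant by name, that simulates dangling skirt and rope bones.
# Seven skirt chains hang from spine_01 and two rope chains from pelvis, all built from capsules.
# Most skirt links have tight swing and wider twist limits, while the B, D, H and J roots move freely.
# Pelvis, spine, thigh, calf and foot capsules supply body volumes, with free constraints on the left leg only.
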PhysicsAsset /Game/Characters/Player/Male/Medium/Bodies/M_MED_CoyoteTrail/Meshes/M_MED_CoyoteTrail_Dyn_Emoting
bodies {
 calf_l [{
   shapeType=Capsule
   length=42.1252
   radius=10.0603
   shapeName=calf_l_capsule
   center=<-23.6951 -2.8429 1.4241>
   rotation=<rot -90 -179.1048 -5.8953>
  }]
 calf_r [{
   shapeType=Capsule
   length=40.3752
   radius=9.5603
   shapeName=calf_l_capsule
   center=<23.6444 2.2632 -2.0946>
   rotation=<rot -90 -96.7215 91.7214>
  }]
 dyn_Rope_l_01 [{
   shapeType=Capsule
   length=8.755
   radius=1
   shapeName=dyn_Rope_l_01_capsule
   center=<0.2049 0.7771 -4.514>
   rotation=<rot -6.8919 -90 -4.9969>
  }]
 dyn_Rope_l_02 [{
   shapeType=Capsule
   length=8.505
   radius=1.25
   shapeName=dyn_Rope_l_02_capsule
   center=<0.1347 -0.0808 -1.8145>
   rotation=<rot -4.2969 -90 -3.4316>
  }]
 dyn_Rope_l_03 [{
   shapeType=Capsule
   length=8.505
   radius=1.25
   shapeName=dyn_Rope_l_03_capsule
   center=<0.24 0.4129 -4.4083>
   rotation=<rot -3.1258 -90 -1.5165>
  }]
 dyn_Rope_r_01 [{
   shapeType=Capsule
   length=8.755
   radius=1
   shapeName=dyn_Rope_r_01_capsule
   center=<-0.0559 0.7771 -4.514>
   rotation=<rot -6.8919 -90 4.9969>
  }]
 dyn_Rope_r_02 [{
   shapeType=Capsule
   length=8.505
   radius=1.25
   shapeName=dyn_Rope_r_02_capsule
   center=<0.0143 -0.0808 -1.8145>
   rotation=<rot -4.2969 -90 3.4316>
  }]
 dyn_Rope_r_03 [{
   shapeType=Capsule
   length=8.505
   radius=1.25
   shapeName=dyn_Rope_r_03_capsule
   center=<-0.091 0.4129 -4.4083>
   rotation=<rot -3.1258 -90 1.5165>
  }]
 dyn_SkirtB01 [{
   shapeType=Capsule
   length=5.1287
   radius=5.341
   shapeName=dyn_SkirtB03_capsule
   center=<-0.456 1.8879 -3.9053>
   rotation=<rot 1.4396 -152.5844 -177.4954>
  }]
 dyn_SkirtB02 [{
   shapeType=Capsule
   length=5.1287
   radius=5.341
   shapeName=dyn_SkirtB03_capsule
   center=<-1.6277 4.153 -3.7937>
   rotation=<rot 1.4396 -152.5844 -177.4954>
  }]
 dyn_SkirtB03 [{shapeType=Capsule, length=9.6287, radius=9.341, shapeName=dyn_SkirtB03_capsule, center=<0 6 -5>, rotation=<rot -2.8652 109.8974 94.0994>}]
 dyn_SkirtB04 [{
   shapeType=Capsule
   length=16.9785
   radius=11.3147
   shapeName=dyn_SkirtB04_capsule
   center=<4 11 -8>
   rotation=<rot -0.2473 -62.4064 -90.1288>
  }]
 dyn_SkirtB05 [{
   shapeType=Capsule
   length=17.3573
   radius=10.9358
   shapeName=dyn_SkirtB05_capsule
   center=<4 11 -8>
   rotation=<rot -0.2473 -62.4064 -90.1288>
  }]
 dyn_SkirtB06 [{
   shapeType=Capsule
   length=15.5126
   radius=11.0114
   shapeName=dyn_SkirtB06_capsule
   center=<4 11 -8>
   rotation=<rot -0.2474 -62.4064 -111.3553>
  }]
 dyn_SkirtD01 [{shapeType=Capsule, length=5.75, radius=5.5, shapeName=dyn_SkirtD01_capsule, center=<0 0 -5.1609>}]
 dyn_SkirtD02 [{shapeType=Capsule, length=10.9554, radius=6.8793, shapeName=dyn_SkirtD02_capsule, center=<0 9 -6>, rotation=<rot -4.9999 -180 -175>}]
 dyn_SkirtD03 [{shapeType=Capsule, length=22.7998, radius=7.7433, shapeName=dyn_SkirtD03_capsule, center=<0 9 -6>, rotation=<rot -10.2785 -90.39 -89.9982>}]
 dyn_SkirtD04 [{shapeType=Capsule, length=21.2998, radius=8.7433, shapeName=dyn_SkirtD04_capsule, center=<0 9 -6>, rotation=<rot -0.2785 -90.3896 -89.9979>}]
 dyn_SkirtD05 [{shapeType=Capsule, length=26.7998, radius=9.7433, shapeName=dyn_SkirtD05_capsule, center=<0 9 -6>, rotation=<rot -0.2785 -90.3896 -89.9979>}]
 dyn_SkirtD06 [{shapeType=Capsule, length=27.0498, radius=9.4933, shapeName=dyn_SkirtD06_capsule, center=<0 9 -6>, rotation=<rot -0.2785 -90.3896 -89.9979>}]
 dyn_SkirtE01 [{shapeType=Capsule, length=5.75, radius=5.5, shapeName=dyn_SkirtE01_capsule, center=<0 0 -5.1609>}]
 dyn_SkirtE02 [{
   shapeType=Capsule
   length=3.8716
   radius=8.1284
   shapeName=dyn_SkirtE02_capsule
   center=<0 4.4233 -4.485>
   rotation=<rot 83.066 -118.404 -118.2284>
  }]
 dyn_SkirtE03 [{shapeType=Capsule, length=22.6447, radius=7.3985, shapeName=dyn_SkirtE03_capsule, center=<0 7 -5>, rotation=<rot -0.2784 -90.3896 -89.998>}]
 dyn_SkirtE04 [{shapeType=Capsule, length=23.5, radius=6.5431, shapeName=dyn_SkirtE03_capsule, center=<0 7 -5>, rotation=<rot -0.2784 -90.3896 -89.998>}]
 dyn_SkirtE05 [{shapeType=Capsule, length=26.2202, radius=10.3229, shapeName=dyn_SkirtE05_capsule, center=<0 7 -5>, rotation=<rot -0.2784 -90.3896 -89.998>}]
 dyn_SkirtE06 [{shapeType=Capsule, length=27.7254, radius=11.3178, shapeName=dyn_SkirtE06_capsule, center=<0 7 -5>, rotation=<rot -0.2784 -90.3897 -89.998>}]
 dyn_SkirtF01 [{shapeType=Capsule, length=5.75, radius=5.5, shapeName=dyn_SkirtF01_capsule, center=<0 0 -5.1609>}]
 dyn_SkirtF02 [{
   shapeType=Capsule
   length=3.5343
   radius=8.4657
   shapeName=dyn_SkirtF02_capsule
   center=<0 4.4233 -4.4851>
   rotation=<rot 83.7383 -77.1354 -77.2096>
  }]
 dyn_SkirtF03 [{shapeType=Capsule, length=21.8947, radius=8.1485, shapeName=dyn_SkirtF03_capsule, center=<0 7 -6>, rotation=<rot -0.2784 -90.3896 -89.9981>}]
 dyn_SkirtF04 [{shapeType=Capsule, length=23.5, radius=6.5431, shapeName=dyn_SkirtF04_capsule, center=<0 7 -6>, rotation=<rot -0.2784 -90.3896 -89.9981>}]
 dyn_SkirtG01 [{shapeType=Capsule, length=5.75, radius=5.5, shapeName=dyn_SkirtG01_capsule, center=<0 0 -5.1609>}]
 dyn_SkirtG02 [{
   shapeType=Capsule
   length=3.8716
   radius=8.1284
   shapeName=dyn_SkirtG02_capsule
   center=<0 4.4233 -4.4851>
   rotation=<rot 83.8498 -83.075 -83.1145>
  }]
 dyn_SkirtG03 [{shapeType=Capsule, length=22.1447, radius=7.8985, shapeName=dyn_SkirtG03_capsule, center=<0 7 -5>, rotation=<rot -0.2784 -90.3896 -89.9982>}]
 dyn_SkirtG04 [{shapeType=Capsule, length=23.5, radius=6.5431, shapeName=dyn_SkirtG04_capsule, center=<0 7 -5>, rotation=<rot -0.2784 -90.3896 -89.9982>}]
 dyn_SkirtG05 [{shapeType=Capsule, length=26.2202, radius=10.3229, shapeName=dyn_SkirtG05_capsule, center=<0 7 -5>, rotation=<rot -0.2784 -90.3896 -89.9982>}]
 dyn_SkirtG06 [{shapeType=Capsule, length=29.7254, radius=11.3178, shapeName=dyn_SkirtG06_capsule, center=<0 7 -5>, rotation=<rot -0.2784 -90.3896 -89.9982>}]
 dyn_SkirtH01 [{shapeType=Capsule, length=5.75, radius=5.5, shapeName=dyn_SkirtH01_capsule, center=<0 0 -5.1609>}]
 dyn_SkirtH02 [{shapeType=Capsule, length=11.2054, radius=6.6293, shapeName=dyn_SkirtD02_capsule, center=<0 9 -6>, rotation=<rot -4.9999 -180 -175>}]
 dyn_SkirtH03 [{shapeType=Capsule, length=22.2998, radius=8.2433, shapeName=dyn_SkirtD03_capsule, center=<0 9 -6>, rotation=<rot -10.2785 -90.39 -89.9982>}]
 dyn_SkirtH04 [{shapeType=Capsule, length=21.2998, radius=8.7433, shapeName=dyn_SkirtD04_capsule, center=<0 9 -6>, rotation=<rot -0.2785 -90.3896 -89.9979>}]
 dyn_SkirtH05 [{shapeType=Capsule, length=27.0498, radius=9.4933, shapeName=dyn_SkirtH05_capsule, center=<0 9 -6>, rotation=<rot -0.2785 -90.3897 -89.9983>}]
 dyn_SkirtH06 [{shapeType=Capsule, length=27.0498, radius=9.4933, shapeName=dyn_SkirtH06_capsule, center=<0 9 -6>, rotation=<rot -0.2785 -90.3897 -89.9983>}]
 dyn_SkirtJ01 [{shapeType=Capsule, length=5.75, radius=5.5, shapeName=dyn_SkirtJ01_capsule, center=<0 0 -5.1609>}]
 dyn_SkirtJ02 [{shapeType=Capsule, length=5.75, radius=5.5, shapeName=dyn_SkirtJ01_capsule, center=<1.7465 0.9653 -5.1609>}]
 dyn_SkirtJ03 [{shapeType=Capsule, length=9.3787, radius=9.591, shapeName=dyn_SkirtB03_capsule, center=<0 6 -5>, rotation=<rot 0.0002 75 95.0002>}]
 dyn_SkirtJ04 [{shapeType=Capsule, length=10.75, radius=11.75, shapeName=dyn_SkirtJ04_capsule, center=<-1 10 -10>, rotation=<rot 90 -143.3538 -116.4449>}]
 dyn_SkirtJ05 [{shapeType=Capsule, length=16.5, radius=11.25, shapeName=dyn_SkirtJ05_capsule, center=<-4 11 -10>, rotation=<rot 90 -143.3465 -121.4375>}]
 dyn_SkirtJ06 [{shapeType=Capsule, length=16.75, radius=11, shapeName=dyn_SkirtJ06_capsule, center=<-4 11 -10>, rotation=<rot 67.6926 -25.0069 1.9018>}]
 dyn_pelvis [{shapeType=Capsule, length=18.25, radius=16.5, shapeName=dyn_pelvis_capsule, center=<2.9551 -7.301 0>, rotation=<rot 90 0 0.0002>}]
 foot_l [{shapeType=Capsule, length=11.2077, radius=10.8588, shapeName=foot_r_capsule, center=<-3.5892 7.5638 0.9746>, rotation=<rot 0 0 -97.3438>}]
 foot_r [{shapeType=Capsule, length=11.2077, radius=10.8588, shapeName=foot_r_capsule, center=<3.5892 -7.5638 -0.9746>, rotation=<rot 0 0 82.6562>}]
 pelvis [{shapeType=Capsule, length=8.4555, radius=11.5757, shapeName=pelvis_capsule, center=<-8.291 0.2611 0>}]
 spine_01 [{shapeType=Capsule, length=3.5, radius=12.5182, shapeName=spine_01_capsule, center=<-3.7991 2.4374 0>}]
 thigh_twist_01_l [{
   shapeType=Capsule
   length=30.1096
   radius=11.1404
   shapeName=thigh_r_capsule
   center=<7.5046 -1.3921 2.828>
   rotation=<rot -88.9708 -0.0007 -179.9994>
  }]
 thigh_twist_01_r [{
   shapeType=Capsule
   length=30.1096
   radius=11.1404
   shapeName=thigh_r_capsule
   center=<-7.5046 1.3921 -2.828>
   rotation=<rot -88.9708 -0.0007 0.0006>
  }]
}
constraints [
 {bone1Name=dyn_SkirtJ02, bone2Name=dyn_SkirtJ01, swing1LimitDegrees=30, swing2LimitDegrees=30, twistLimitDegrees=60}
 {bone1Name=dyn_SkirtJ03, bone2Name=dyn_SkirtJ02, swing1LimitDegrees=2, swing2LimitDegrees=10, twistLimitDegrees=60}
 {bone1Name=dyn_SkirtJ04, bone2Name=dyn_SkirtJ03, swing1LimitDegrees=2, swing2LimitDegrees=10, twistLimitDegrees=60}
 {bone1Name=dyn_SkirtJ05, bone2Name=dyn_SkirtJ04, swing1LimitDegrees=2, swing2LimitDegrees=10, twistLimitDegrees=60}
 {bone1Name=dyn_SkirtJ06, bone2Name=dyn_SkirtJ05, swing1LimitDegrees=2, swing2LimitDegrees=10, twistLimitDegrees=60}
 {bone1Name=dyn_SkirtH02, bone2Name=dyn_SkirtH01, swing1LimitDegrees=2, swing2LimitDegrees=15, twistLimitDegrees=45}
 {bone1Name=dyn_SkirtH03, bone2Name=dyn_SkirtH02, swing1LimitDegrees=2, swing2LimitDegrees=10, twistLimitDegrees=60}
 {bone1Name=dyn_SkirtH04, bone2Name=dyn_SkirtH03, swing1LimitDegrees=2, swing2LimitDegrees=10, twistLimitDegrees=60}
 {bone1Name=dyn_SkirtH05, bone2Name=dyn_SkirtH04, swing1LimitDegrees=2, swing2LimitDegrees=10, twistLimitDegrees=60}
 {bone1Name=dyn_SkirtH06, bone2Name=dyn_SkirtH05, swing1LimitDegrees=2, swing2LimitDegrees=10, twistLimitDegrees=60}
 {bone1Name=dyn_SkirtG02, bone2Name=dyn_SkirtG01, swing1LimitDegrees=8, swing2LimitDegrees=8, twistLimitDegrees=20}
 {bone1Name=dyn_SkirtG03, bone2Name=dyn_SkirtG02, swing1LimitDegrees=2, swing2LimitDegrees=10, twistLimitDegrees=60}
 {bone1Name=dyn_SkirtG04, bone2Name=dyn_SkirtG03, swing1LimitDegrees=2, swing2LimitDegrees=10, twistLimitDegrees=60}
 {bone1Name=dyn_SkirtG05, bone2Name=dyn_SkirtG04, swing1LimitDegrees=2, swing2LimitDegrees=10, twistLimitDegrees=60}
 {bone1Name=dyn_SkirtG06, bone2Name=dyn_SkirtG05, swing1LimitDegrees=2, swing2LimitDegrees=10, twistLimitDegrees=60}
 {bone1Name=dyn_SkirtF02, bone2Name=dyn_SkirtF01, swing1LimitDegrees=8, swing2LimitDegrees=8, twistLimitDegrees=25}
 {bone1Name=dyn_SkirtF03, bone2Name=dyn_SkirtF02, swing1LimitDegrees=2, swing2LimitDegrees=2, twistLimitDegrees=60}
 {bone1Name=dyn_SkirtF04, bone2Name=dyn_SkirtF03, swing1LimitDegrees=2, swing2LimitDegrees=2, twistLimitDegrees=60}
 {bone1Name=dyn_SkirtE02, bone2Name=dyn_SkirtE01, swing1LimitDegrees=8, swing2LimitDegrees=8, twistLimitDegrees=20}
 {bone1Name=dyn_SkirtE03, bone2Name=dyn_SkirtE02, swing1LimitDegrees=2, swing2LimitDegrees=10, twistLimitDegrees=60}
 {bone1Name=dyn_SkirtE04, bone2Name=dyn_SkirtE03, swing1LimitDegrees=2, swing2LimitDegrees=10, twistLimitDegrees=60}
 {bone1Name=dyn_SkirtE05, bone2Name=dyn_SkirtE04, swing1LimitDegrees=2, swing2LimitDegrees=10, twistLimitDegrees=60}
 {bone1Name=dyn_SkirtE06, bone2Name=dyn_SkirtE05, swing1LimitDegrees=2, swing2LimitDegrees=10, twistLimitDegrees=60}
 {bone1Name=dyn_SkirtD02, bone2Name=dyn_SkirtD01, swing1LimitDegrees=2, swing2LimitDegrees=10, twistLimitDegrees=60}
 {bone1Name=dyn_SkirtD03, bone2Name=dyn_SkirtD02, swing1LimitDegrees=2, swing2LimitDegrees=10, twistLimitDegrees=60}
 {bone1Name=dyn_SkirtD04, bone2Name=dyn_SkirtD03, swing1LimitDegrees=2, swing2LimitDegrees=10, twistLimitDegrees=60}
 {bone1Name=dyn_SkirtD05, bone2Name=dyn_SkirtD04, swing1LimitDegrees=2, swing2LimitDegrees=10, twistLimitDegrees=60}
 {bone1Name=dyn_SkirtD06, bone2Name=dyn_SkirtD05, swing1LimitDegrees=2, swing2LimitDegrees=10, twistLimitDegrees=60}
 {
  bone1Name=dyn_SkirtB02
  bone2Name=dyn_SkirtB01
  swing1LimitDegrees=30
  swing1Motion=Free
  swing2LimitDegrees=30
  swing2Motion=Free
  twistLimitDegrees=60
  twistMotion=Free
 }
 {bone1Name=dyn_SkirtB03, bone2Name=dyn_SkirtB02, swing1LimitDegrees=2, swing2LimitDegrees=10, twistLimitDegrees=60}
 {bone1Name=dyn_SkirtB04, bone2Name=dyn_SkirtB03, swing1LimitDegrees=2, swing2LimitDegrees=10, twistLimitDegrees=60}
 {bone1Name=dyn_SkirtB05, bone2Name=dyn_SkirtB04, swing1LimitDegrees=2, swing2LimitDegrees=10, twistLimitDegrees=60}
 {bone1Name=dyn_SkirtB06, bone2Name=dyn_SkirtB05, swing1LimitDegrees=2, swing2LimitDegrees=10, twistLimitDegrees=60}
 {bone1Name=dyn_Rope_l_01, bone2Name=pelvis, swing1LimitDegrees=1, swing2LimitDegrees=1, twistLimitDegrees=10}
 {bone1Name=dyn_Rope_l_02, bone2Name=dyn_Rope_l_01, swing1LimitDegrees=15, swing2LimitDegrees=15, twistLimitDegrees=45}
 {bone1Name=dyn_Rope_l_03, bone2Name=dyn_Rope_l_02, swing1LimitDegrees=15, swing2LimitDegrees=15, twistLimitDegrees=45}
 {bone1Name=dyn_Rope_r_01, bone2Name=pelvis, swing1LimitDegrees=1, swing2LimitDegrees=1, twistLimitDegrees=10}
 {bone1Name=dyn_Rope_r_02, bone2Name=dyn_Rope_r_01, swing1LimitDegrees=25, swing2LimitDegrees=25, twistLimitDegrees=45}
 {bone1Name=dyn_Rope_r_03, bone2Name=dyn_Rope_r_02, swing1LimitDegrees=25, swing2LimitDegrees=25, twistLimitDegrees=45}
 {
  bone1Name=dyn_SkirtB01
  bone2Name=spine_01
  swing1LimitDegrees=5
  swing1Motion=Free
  swing2LimitDegrees=5
  swing2Motion=Free
  twistLimitDegrees=20
  twistMotion=Free
 }
 {
  bone1Name=dyn_SkirtD01
  bone2Name=spine_01
  swing1LimitDegrees=45
  swing1Motion=Free
  swing2LimitDegrees=45
  swing2Motion=Free
  twistLimitDegrees=45
  twistMotion=Free
 }
 {bone1Name=dyn_SkirtE01, bone2Name=spine_01, swing1LimitDegrees=5, swing2LimitDegrees=5, twistLimitDegrees=10}
 {bone1Name=dyn_SkirtF01, bone2Name=spine_01, swing1LimitDegrees=5, swing2LimitDegrees=5, twistLimitDegrees=10}
 {bone1Name=dyn_SkirtG01, bone2Name=spine_01, swing1LimitDegrees=5, swing2LimitDegrees=5, twistLimitDegrees=10}
 {
  bone1Name=dyn_SkirtH01
  bone2Name=spine_01
  swing1LimitDegrees=45
  swing1Motion=Free
  swing2LimitDegrees=45
  swing2Motion=Free
  twistLimitDegrees=45
  twistMotion=Free
 }
 {
  bone1Name=dyn_SkirtJ01
  bone2Name=spine_01
  swing1LimitDegrees=45
  swing1Motion=Free
  swing2LimitDegrees=45
  swing2Motion=Free
  twistLimitDegrees=45
  twistMotion=Free
 }
 {
  bone1Name=calf_l
  bone2Name=pelvis
  swing1LimitDegrees=45
  swing1Motion=Free
  swing2LimitDegrees=45
  swing2Motion=Free
  twistLimitDegrees=45
  twistMotion=Free
 }
 {
  bone1Name=foot_l
  bone2Name=calf_l
  swing1LimitDegrees=45
  swing1Motion=Free
  swing2LimitDegrees=45
  swing2Motion=Free
  twistLimitDegrees=45
  twistMotion=Free
 }
 {
  bone1Name=thigh_twist_01_l
  bone2Name=pelvis
  swing1LimitDegrees=45
  swing1Motion=Free
  swing2LimitDegrees=45
  swing2Motion=Free
  twistLimitDegrees=45
  twistMotion=Free
 }
]
